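# Niagara VFX for the DegreeProper emote: warm orange spotlights from above with soft floating sparkles.
# SoftSparkles drifts sprites with curl noise, flickering brightness and material-parameter fades.
# TopCenter and TopSide burst looping mesh-and-sprite lights in three spawn groups that set VisibilityTag, with TopSide mirroring a pair via partitioning and aligning meshes along their offsets.
# All emitters fade in by system age, apply time-of-day scaling and run BlendSpace.
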NiagaraSystem /BRCosmetics/Animation/Game/MainPlayer/Emotes/DegreeProper/VFX/NS_DegreeProper
dependencies {
 usedDataInterfaces [&NiagaraDataInterfaceArrayColor, &NiagaraDataInterfaceCurve, &NiagaraDataInterfaceArrayFloat]
 usedDynamicInputs [&Power, &Divide_Float]
 usedModules [&SystemState, &BlendSpaceHelper]
}
emitters {
 SoftSparkles {
  emitterUpdateScript [
   &EmitterState()
   &SpawnRate(SpawnRate=3, Use Spawn Probability=false)
  ]
  particleSpawnScript [
   &InitializeParticle(
    Color=&SelectLinearColorFromArray(
     Array Sampling Mode=::NewEnumerator0
     Direct Array Index=?
     Color Selection Array=DataInterface() {propertyValues {ColorData [<rgba 1 0.3769 0.0076 1>, <rgba 1 0.1713 0 1>, <rgba 1 0.0225 0.0241 1>]}}
     Random Seed=0
    )
    Lifetime Max=2.4
    Lifetime Min=1.2
    Lifetime Mode=::NewEnumerator1
    Sprite Size Mode=::NewEnumerator1
    Uniform Sprite Size Max=44
    Uniform Sprite Size Min=22
    UsePositionOffset=false
    Write Material Random=true
   )
   &ShapeLocation(MaskSpawnGroup=false, Offset=<0 0 80>, Offset Mode=::NewEnumerator0, Override Seed=false, UseOwnerScale=true)
   /SetVariables(Particles.Curve:NiagaraFloat=&FloatFromCurve(FloatCurve=DataInterface() {propertyValues {Curve {keys [
         {interpMode=RCIM_Cubic, leaveTangent=7.0743, tangentMode=RCTM_Break, time=0, value=0.35}
         {interpMode=RCIM_Cubic, tangentMode=RCTM_Break, time=0.2, value=0.95}
         {interpMode=RCIM_Cubic, leaveTangent=-0.1241, tangentMode=RCTM_Break, time=0.4, value=0.55}
         {interpMode=RCIM_Cubic, tangentMode=RCTM_Break, time=0.6, value=0.8}
         {interpMode=RCIM_Cubic, tangentMode=RCTM_Break, time=0.8, value=0.6}
         {arriveTangent=2.6513, interpMode=RCIM_Cubic, tangentMode=RCTM_Break, time=1, value=1}
        ]}}})) {isSetParameters=true}
  ]
  particleUpdateScript [
   &ParticleState()
   &CurlNoiseForce(
    Noise Frequency=100
    Noise Strength=&FloatFromCurve(
     FloatCurve=DataInterface() {propertyValues {Curve {keys [{tangentMode=RCTM_User, time=0, value=0}, {tangentMode=RCTM_User, time=1, value=1}]}}}
     Scale Curve=111
    )
    Pan Noise=false
   )
   &SolveForcesAndVelocity(Clamp Velocity=false, Limit Acceleration=false)
   &DynamicMaterialParameters(
    Index 0 Param 1=&Multiply_Float(
     A=&FloatFromCurve(FloatCurve=DataInterface() {propertyValues {Curve {keys [
          {interpMode=RCIM_Cubic, leaveTangent=0.1514, tangentMode=RCTM_Break, time=0, value=1}
          {arriveTangent=-0.3559, interpMode=RCIM_Cubic, tangentMode=RCTM_Break, time=1, value=0}
         ]}}})
     B=&Lerp_Float(A=1, B=0, Alpha=$Particles.Curve)
    )
    Module.Param0WriteEnabled=true
    Module.Param1WriteEnabled=true
    Module.Param2WriteEnabled=true
    Module.Param3WriteEnabled=true
   )
   &ScaleColor(
    Scale Alpha=&FloatFromCurve(
     FloatCurve=DataInterface() {propertyValues {Curve {keys [
         {interpMode=RCIM_Cubic, tangentMode=RCTM_User, time=0, value=0}
         {interpMode=RCIM_Cubic, tangentMode=RCTM_User, time=0.5, value=1}
         {tangentMode=RCTM_User, time=1, value=0}
        ]}}}
     Scale Curve=0.2
    )
    Scale RGB=&VectorFromFloat(Value=&Lerp_Float(
      A=1
      B=5
      Alpha=&Sine(Normalized Angle=$Emitter.Age, Period=&RandomRangeFloat(Minimum=0.1, Maximum=0.4), Scale=1, Bias=1)
     ))
    ScaleA=true
    ScaleRGB=true
   )
   &ScaleSpriteSize(
    Scale Sprite Size Mode=::NewEnumerator0
    Uniform Scale Factor=&Multiply_Float(
     A=&FloatFromCurve(FloatCurve=DataInterface() {propertyValues {Curve {keys [
          {interpMode=RCIM_Cubic, leaveTangent=2.1941, tangentMode=RCTM_Break, time=0, value=0.35}
          {arriveTangent=0.204, interpMode=RCIM_Cubic, tangentMode=RCTM_Break, time=1, value=1}
         ]}}})
     B=$Particles.Curve
    )
   )
   &ScaleColor(
    Scale Alpha=&FNTimeOfDayScalar(Night Scale=1, Morning Scale=1, Day Scale=1.2, Evening Scale=1)
    Scale RGB=&VectorFromFloat(Value=&FNTimeOfDayScalar(Night Scale=1, Morning Scale=1, Day Scale=1.4, Evening Scale=1))
    ScaleA=true
    ScaleRGB=true
   ) {module=ScaleColor001}
   &ScaleColor(
    Scale Alpha=&FloatFromCurve(
     FloatCurve=DataInterface() {propertyValues {Curve {keys [{interpMode=RCIM_Cubic, tangentMode=RCTM_Break, time=0, value=0}, {interpMode=RCIM_Cubic, tangentMode=RCTM_Break, time=1, value=1}]}}}
     CurveIndex=$System.Age
    )
    ScaleA=true
    ScaleRGB=true
   ) {module=ScaleColor002}
   &BlendSpace(Rotation Contribution=0)
  ]
  rendererClasses [&NiagaraSpriteRendererProperties]
  renderers [{rendererClass=&NiagaraSpriteRendererProperties}]
 }
 SpotLight_TopCenter {
  emitterUpdateScript [
   &EmitterState(
    Inactive Response=::NewEnumerator1
    Life Cycle Mode=::NewEnumerator1
    Loop Behavior=::NewEnumerator1
    Loop Duration Mode=::NewEnumerator1
    UseLoopDelay=false
   )
   &SpawnBurst_Instantaneous(Use Spawn Probability=false, UseLoopCountLimit=false)
   &SpawnBurst_Instantaneous(Spawn Group=1, Use Spawn Probability=false, UseLoopCountLimit=false) {module=SpawnBurst_Instantaneous001}
   &SpawnBurst_Instantaneous(Spawn Group=2, Use Spawn Probability=false, UseLoopCountLimit=false) {module=SpawnBurst_Instantaneous002}
  ]
  particleSpawnScript [
   &InitializeParticle(
    Color=&MaskLinearColorBySpawnGroup(Masked Linear Color=<rgba 1 0.049 0 1>, Passthrough Linear Color=<rgba 1 0.2239 0 1>)
    Mesh Scale=<1.2 1.2 1.2>
    Mesh Scale Mode=::NewEnumerator4
    Position Offset=<0 0 340>
    Sprite Rotation Mode=::NewEnumerator2
    Sprite Size Mode=::NewEnumerator3
    Uniform Sprite Size=200
    UsePositionOffset=true
    Write Material Random=true
   )
   /SetVariables(Particles.VisibilityTag=&SpawnGroup()) {isSetParameters=true, module=SetVariables_#1}
   /SetVariables(Particles.FlareRotation:NiagaraFloat=&RandomRangeFloat(Minimum=15, Maximum=20)) {isSetParameters=true, module=SetVariables_#2}
  ]
  particleUpdateScript [
   &ParticleState(
    Kill Particles When Lifetime Has Elapsed=false
    Let Infinitely Lived Particles Die When Emitter Deactivates=true
    Loop Particles Lifetime=true
   )
   &SpriteRotationRate(Rotation Rate=$Particles.FlareRotation)
   &ScaleColor(
    Scale Alpha=0.75
    Scale RGB=&VectorFromFloat(Value=&FloatFromCurve(
      FloatCurve=DataInterface() {propertyValues {Curve {keys [
          {interpMode=RCIM_Cubic, tangentMode=RCTM_Break, time=0, value=1}
          {interpMode=RCIM_Cubic, tangentMode=RCTM_Break, time=0.2, value=1.2}
          {interpMode=RCIM_Cubic, tangentMode=RCTM_Break, time=0.4, value=1}
          {interpMode=RCIM_Cubic, tangentMode=RCTM_Break, time=0.5, value=1}
          {interpMode=RCIM_Cubic, tangentMode=RCTM_Break, time=0.6, value=1}
          {interpMode=RCIM_Cubic, tangentMode=RCTM_Break, time=0.8, value=1.2}
          {interpMode=RCIM_Cubic, tangentMode=RCTM_Break, time=1, value=1}
         ]}}}
      CurveIndex=&Add_Float(
       A=$Particles.NormalizedAge
       B=&SelectFloatFromArray(
        Direct Array Index=&SpawnGroup()
        Float Selection Array=DataInterface() {propertyValues {FloatData [0.03,0.04,0.025,0.02,0.035,0.027]}}
       )
      )
     ))
    ScaleA=true
    ScaleRGB=true
   ) {module=ScaleColor001}
   /SetVariables(
    Particles.GlowCameraOffset=30
    Particles.GlowColor=&MakeLinearColor(
     R:NiagaraFloat=3
     G:NiagaraFloat=0.147
     B:NiagaraFloat=0
     A:NiagaraFloat=&Multiply_Float(
      A=1
      B=&FloatFromCurve(
       FloatCurve=DataInterface() {propertyValues {Curve {keys [{interpMode=RCIM_Cubic, tangentMode=RCTM_Break, time=0, value=0}, {interpMode=RCIM_Cubic, tangentMode=RCTM_Break, time=0.6, value=1}]}}}
       CurveIndex=$System.Age
       Scale Curve=0.2
      )
     )
    )
   ) {isSetParameters=true, module=SetVariables_#3}
   &ScaleColor(
    Scale Alpha=&FNTimeOfDayScalar(Night Scale=0.65, Morning Scale=1, Evening Scale=1)
    Scale RGB=&VectorFromFloat(Value=&FNTimeOfDayScalar(Night Scale=1, Morning Scale=1, Day Scale=1.5, Evening Scale=1))
    ScaleA=true
    ScaleRGB=true
   )
   &ScaleColor(
    Scale Alpha=&FloatFromCurve(
     FloatCurve=DataInterface() {propertyValues {Curve {keys [{interpMode=RCIM_Cubic, tangentMode=RCTM_Break, time=0, value=0}, {interpMode=RCIM_Cubic, tangentMode=RCTM_Break, time=1, value=1}]}}}
     CurveIndex=$System.Age
    )
    ScaleA=true
    ScaleRGB=true
   ) {module=ScaleColor002}
   &BlendSpace(Rotation Contribution=0)
  ]
  rendererClasses [&NiagaraMeshRendererProperties, &NiagaraSpriteRendererProperties]
  renderers [
   {rendererClass=&NiagaraMeshRendererProperties, rendererIndex=0}
   {rendererClass=&NiagaraMeshRendererProperties, rendererIndex=1}
   {rendererClass=&NiagaraSpriteRendererProperties, rendererIndex=2}
  ]
 }
 SpotLight_TopSide {
  emitterUpdateScript [
   &EmitterState(
    Inactive Response=::NewEnumerator1
    Life Cycle Mode=::NewEnumerator1
    Loop Behavior=::NewEnumerator1
    Loop Duration Mode=::NewEnumerator1
    UseLoopDelay=false
   )
   &SpawnBurst_Instantaneous(Spawn Count=2, Use Spawn Probability=false, UseLoopCountLimit=false)
   &SpawnBurst_Instantaneous(Spawn Count=2, Spawn Group=1, Use Spawn Probability=false, UseLoopCountLimit=false) {module=SpawnBurst_Instantaneous001}
   &SpawnBurst_Instantaneous(Spawn Count=2, Spawn Group=2, Use Spawn Probability=false, UseLoopCountLimit=false) {module=SpawnBurst_Instantaneous002}
  ]
  particleSpawnScript [
   &PartitionParticles()
   &InitializeParticle(
    Color=&MaskLinearColorBySpawnGroup(Masked Linear Color=<rgba 1 0.049 0 1>, Passthrough Linear Color=<rgba 1 0.2239 0 1>)
    Mesh Scale=<1.1 1.1 1>
    Mesh Scale Mode=::NewEnumerator4
    Sprite Rotation Mode=::NewEnumerator2
    Sprite Size Mode=::NewEnumerator3
    Uniform Sprite Size=200
    UsePositionOffset=false
    Write Material Random=true
   )
   /SetVariables(Particles.VisibilityTag=&SpawnGroup()) {isSetParameters=true, module=SetVariables_#1}
   /SetVariables(Particles.FlareRotation:NiagaraFloat=&RandomRangeFloat(Minimum=15, Maximum=20)) {isSetParameters=true, module=SetVariables_#2}
   /SetVariables(Particles.Position=&AddVectorToPosition(
     Position=$Engine.Emitter.SimulationPosition
     Vector=&TransformVector(
      Vector=&MaskVector(Mask Source=1, Mask Parameter=$Particles.PartitionParticles.Partition, Mask=0, Masked Vector=<-136 55 290>, Passthrough Vector=<136 55 290>)
      Destination Space=::World
     )
    )) {isSetParameters=true}
   &InitialMeshOrientation(
    Mesh Axis To Orient=<0 0 1>
    Mesh Orientation Mode=::NewEnumerator2
    Orientation Vector=&Subtract_Position(A=$Particles.Position, B=$Engine.Emitter.SimulationPosition)
    Use Rotation Vector=false
   ) {module=InitialMeshOrientation001}
  ]
  particleUpdateScript [
   &ParticleState(
    Kill Particles When Lifetime Has Elapsed=false
    Let Infinitely Lived Particles Die When Emitter Deactivates=true
    Loop Particles Lifetime=true
   )
   &SpriteRotationRate(Rotation Rate=$Particles.FlareRotation)
   &ScaleColor(
    Scale Alpha=0.5
    Scale RGB=&VectorFromFloat(Value=&FloatFromCurve(
      FloatCurve=DataInterface() {propertyValues {Curve {keys [
          {interpMode=RCIM_Cubic, tangentMode=RCTM_Break, time=0, value=1}
          {interpMode=RCIM_Cubic, tangentMode=RCTM_Break, time=0.2, value=1.2}
          {interpMode=RCIM_Cubic, tangentMode=RCTM_Break, time=0.4, value=1}
          {interpMode=RCIM_Cubic, tangentMode=RCTM_Break, time=0.5, value=1}
          {interpMode=RCIM_Cubic, tangentMode=RCTM_Break, time=0.6, value=1}
          {interpMode=RCIM_Cubic, tangentMode=RCTM_Break, time=0.8, value=1.2}
          {interpMode=RCIM_Cubic, tangentMode=RCTM_Break, time=1, value=1}
         ]}}}
      CurveIndex=&Add_Float(
       A=$Particles.NormalizedAge
       B=&SelectFloatFromArray(
        Direct Array Index=&SpawnGroup()
        Float Selection Array=DataInterface() {propertyValues {FloatData [0.03,0.04,0.025,0.02,0.035,0.027]}}
       )
      )
     ))
    ScaleA=true
    ScaleRGB=true
   ) {module=ScaleColor001}
   /SetVariables(
    Particles.GlowCameraOffset=30
    Particles.GlowColor=&MakeLinearColor(
     R:NiagaraFloat=3
     G:NiagaraFloat=0.147
     B:NiagaraFloat=0
     A:NiagaraFloat=&Multiply_Float(
      A=1
      B=&FloatFromCurve(
       FloatCurve=DataInterface() {propertyValues {Curve {keys [{interpMode=RCIM_Cubic, tangentMode=RCTM_Break, time=0, value=0}, {interpMode=RCIM_Cubic, tangentMode=RCTM_Break, time=0.6, value=1}]}}}
       CurveIndex=$System.Age
       Scale Curve=0.2
      )
     )
    )
   ) {isSetParameters=true, module=SetVariables_#3}
   &ScaleColor(
    Scale Alpha=&FNTimeOfDayScalar(Night Scale=0.65, Morning Scale=1, Day Scale=1.3, Evening Scale=1)
    ScaleA=true
    ScaleRGB=true
   )
   &ScaleColor(
    Scale Alpha=&FloatFromCurve(
     FloatCurve=DataInterface() {propertyValues {Curve {keys [{interpMode=RCIM_Cubic, tangentMode=RCTM_Break, time=0, value=0}, {interpMode=RCIM_Cubic, tangentMode=RCTM_Break, time=1, value=1}]}}}
     CurveIndex=$System.Age
    )
    ScaleA=true
    ScaleRGB=true
   ) {module=ScaleColor002}
   &BlendSpace(Rotation Contribution=0)
  ]
  rendererClasses [&NiagaraMeshRendererProperties, &NiagaraSpriteRendererProperties]
  renderers [
   {rendererClass=&NiagaraMeshRendererProperties, rendererIndex=0}
   {rendererClass=&NiagaraMeshRendererProperties, rendererIndex=1}
   {rendererClass=&NiagaraSpriteRendererProperties, rendererIndex=2}
  ]
 }
}
summary {
 systemName=NS_DegreeProper
 emitters [
  {emitterName=SoftSparkles, rendererClasses [&NiagaraSpriteRendererProperties]}
  {
   emitterName=SpotLight_TopSide
   rendererClasses [&NiagaraMeshRendererProperties, &NiagaraSpriteRendererProperties]
  }
  {
   emitterName=SpotLight_TopCenter
   rendererClasses [&NiagaraMeshRendererProperties, &NiagaraSpriteRendererProperties]
  }
 ]
 userVariables [
  User.bIsFrontEnd {type=NiagaraBool}
  User.bIsFrontEndPreview {type=NiagaraBool}
 ]
}
